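# Niagara trail effect for the Lifting Aura glider cosmetic, combining straight and spiraling ribbons, a heart glow sprite, a light and speed lines.
# Ribbons, glow and light follow skeletal mesh samples and ramp in with particle age.
# Burst-spawned ribbon points ease toward offset goals with owner-velocity lag, while sine/cosine offsets twist the two spirals.
# Smoothed forward dot drives ribbon length, brightness and UV scroll; front-end modifiers retune straight ribbon and speed lines.
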NiagaraSystem /BRCosmetics/Gadgets/Assets/VinderTech_GliderChute/Glider_LiftingAura/FX/NS_Glider_LiftingAura_Trail
dependencies {
 usedDataInterfaces [
  &NiagaraDataInterfaceColorCurve
  &NiagaraDataInterfaceCurve
  &NiagaraDataInterfaceVector2DCurve
  &NiagaraDataInterfaceSkeletalMesh
 ]
 usedDynamicInputs [&ClampFloat, &SetBoolByFloatComparison]
 usedModules [
  &self:SystemSpawnScript.SystemScriptSource.SystemScriptGraph.NiagaraNodeAssignment_3.SetVariables_#1
  &self:SystemSpawnScript.SystemScriptSource.SystemScriptGraph.NiagaraNodeAssignment_8.SetVariables_#2
  &self:SystemSpawnScript.SystemScriptSource.SystemScriptGraph.NiagaraNodeAssignment_4.SetVariables_#3
  &self:SystemSpawnScript.SystemScriptSource.SystemScriptGraph.NiagaraNodeAssignment_5.SetVariables_#4
  &self:SystemSpawnScript.SystemScriptSource.SystemScriptGraph.NiagaraNodeAssignment_7.SetVariables_#5
  &self:SystemSpawnScript.SystemScriptSource.SystemScriptGraph.NiagaraNodeAssignment_0.SetVariables_#6
  &SystemState
  &self:SystemSpawnScript.SystemScriptSource.SystemScriptGraph.NiagaraNodeAssignment_6.SetVariables_#7
  &FN_Integrator
  &Timeline
  &self:SystemSpawnScript.SystemScriptSource.SystemScriptGraph.NiagaraNodeAssignment_1.SetVariables_#8
 ]
}
emitters {
 NE_HeartGlow {
  emitterUpdateScript [
   &EmitterState(
    Inactive Response=::NewEnumerator1
    Life Cycle Mode=::NewEnumerator1
    Loop Behavior=::NewEnumerator1
    Loop Duration Mode=::NewEnumerator1
    UseLoopDelay=false
   )
   &SpawnBurst_Instantaneous(Use Spawn Probability=false, UseLoopCountLimit=false)
  ]
  particleSpawnScript [&InitializeParticle(
    Color=$User.SpiralColorInner
    Lifetime=0.5
    Sprite Size Mode=::NewEnumerator3
    Uniform Sprite Size=40
    UsePositionOffset=false
    Write Material Random=true
   )]
  particleUpdateScript [
   &ParticleState(Kill Particles When Lifetime Has Elapsed=false)
   &SkeletalMeshLocation(Bone Sampling Mode=::NewEnumerator4, MaskSpawnGroup=false, Skeletal Mesh=$System.Skeletal Mesh)
   /SetVariables(Particles.SpawnAlpha:NiagaraFloat=&FloatFromCurve(FloatCurve=DataInterface() {propertyValues {Curve {keys [<at 0 0>,<at 1 1>]}}}, CurveIndex=$Particles.Age)) {isSetParameters=true, module=SetVariables_#9}
   &DynamicMaterialParameters(
    Index 0 Param 2=1
    Module.Index0Param0Write=true
    Module.Index0Param1Write=true
    Module.Index0Param2Write=true
    Module.Index0Param3Write=true
   )
   &ScaleColor(Scale Alpha=$Particles.SpawnAlpha, ScaleA=true, ScaleRGB=true)
  ]
  rendererClasses [&NiagaraSpriteRendererProperties]
  renderers [{rendererClass=&NiagaraSpriteRendererProperties}]
 }
 NE_Light {
  emitterUpdateScript [
   &EmitterState(
    Inactive Response=::NewEnumerator1
    Life Cycle Mode=::NewEnumerator1
    Loop Behavior=::NewEnumerator1
    Loop Duration Mode=::NewEnumerator1
    UseLoopDelay=false
   )
   &SpawnBurst_Instantaneous(Spawn Time=1, Use Spawn Probability=false, UseLoopCountLimit=false)
  ]
  particleSpawnScript [&InitializeParticle(
    Color=&MultiplyLinearColor(Linear Color=$User.SpiralColorInner, Scale Factor=&MakeVector4_FromFloat(Float=10))
    Lifetime=0.5
    Sprite Size Mode=::NewEnumerator3
    Uniform Sprite Size=40
    UsePositionOffset=false
    Write Material Random=true
   )]
  particleUpdateScript [
   &ParticleState(Kill Particles When Lifetime Has Elapsed=false)
   /SetVariables(Particles.SpawnAlpha:NiagaraFloat=&FloatFromCurve(FloatCurve=DataInterface() {propertyValues {Curve {keys [<at 0 0>,<at 1 1>]}}}, CurveIndex=$Particles.Age)) {isSetParameters=true, module=SetVariables_#9}
   &SkeletalMeshLocation(
    Bone Sampling Mode=::NewEnumerator4
    MaskSpawnGroup=false
    Sampled Position Offset=<0 40 0>
    Sampled Position Offset Coordinate Space:ENiagaraSkelMeshTransforms=::NewEnumerator2
    Skeletal Mesh=$System.Skeletal Mesh
   )
   &FN_Light_Attributes(
    Color=$User.SpiralColorInner
    Create Color Attribute=true
    Create Exponent Attribute=true
    Create Radius Attribute=true
    Exponent=4
    Radius=&Multiply_Float(A=60, B=$Particles.SpawnAlpha)
    Scale Color=50
   )
  ]
  rendererClasses [&NiagaraLightRendererProperties]
  renderers [{rendererClass=&NiagaraLightRendererProperties, rendererIndex=0}]
 }
 NE_SpiralRibbon {
  emitterSpawnScript [/SetVariables(Emitter.ParticlePerRibbon=40, Emitter.RibbonNumber=2) {isSetParameters=true}]
  emitterUpdateScript [
   &EmitterState(
    Inactive Response=::NewEnumerator1
    Life Cycle Mode=::NewEnumerator1
    Loop Behavior=::NewEnumerator1
    Loop Duration Mode=::NewEnumerator1
    UseLoopDelay=false
   )
   &SpawnBurst_Instantaneous(
    Spawn Count=&Multiply_Int(A=$Emitter.ParticlePerRibbon, B=$Emitter.RibbonNumber)
    Spawn Time=1
    Use Spawn Probability=false
    UseLoopCountLimit=false
   )
  ]
  particleSpawnScript [
   &InitializeParticle(Lifetime=5, Ribbon Width=7, Ribbon Width Mode=::NewEnumerator1, UsePositionOffset=false, Write Material Random=true)
   &PartitionParticles(
    Number Of Partitions=$Emitter.RibbonNumber
    Particles Per Partition:NiagaraInt32=$Emitter.ParticlePerRibbon
    Partition Mode:ENiagara_PartitionMode=::NewEnumerator1
   )
   /SetVariables(Particles.NormalizedRibbonPos:NiagaraFloat=&OneMinusFloat(Float=$Particles.PartitionParticles.Normalized Index In Partition)) {isSetParameters=true, module=SetVariables_#10}
   /SetVariables(
    Particles.SpiralFreq=1
    Particles.SpiralStrength:NiagaraFloat=&Multiply_Float(
     A=15
     B=&FloatFromCurve(
      FloatCurve=DataInterface() {propertyValues {Curve {keys [
          {arriveTangent=3.6599, interpMode=RCIM_Cubic, leaveTangent=3.6599, tangentMode=RCTM_User, time=0.3, value=-0.025}
          {interpMode=RCIM_Cubic, tangentMode=RCTM_User, time=0.5, value=1}
          {arriveTangent=-0.1348, interpMode=RCIM_Cubic, leaveTangent=-0.1348, tangentMode=RCTM_User, time=1, value=0}
         ]}}}
      CurveIndex=$Particles.NormalizedRibbonPos
     )
    )
   ) {isSetParameters=true}
   /SetVariables(Particles.RibbonU0Override=$Particles.NormalizedRibbonPos) {isSetParameters=true}
   /SetVariables() {isSetParameters=true, module=SetVariables_#11}
  ]
  particleUpdateScript [
   &ParticleState(Kill Particles When Lifetime Has Elapsed=false)
   &SkeletalMeshLocation(Bone Sampling Mode=::NewEnumerator4, MaskSpawnGroup=false, Position Sampling=::NewEnumerator2, Skeletal Mesh=$System.Skeletal Mesh)
   /SetVariables(Particles.UVOffset:NiagaraFloat=&Add_Float(
     A=$Particles.UVOffset
     B=&Multiply_Float(A=$Engine.DeltaTime, B=&Lerp_Float(A=-0.1, B=-1.5, Alpha=$System.SmoothedForwardDot))
    )) {isSetParameters=true, module=SetVariables_#12}
   &Color(
    Color=&ColorFromCurve(
     ColorCurve=DataInterface() {propertyValues {
       AlphaCurve {keys [
         <at 0.1909 0>
         {arriveTangent=2.4905, interpMode=RCIM_Cubic, leaveTangent=2.4905, time=0.4114, value=1}
         <at 0.5924 1>
         <at 0.8624 0 RCIM_Cubic>
        ]}
       BlueCurve {keys [<at 0.2333 0.3>, {arriveTangent=0.627, interpMode=RCIM_Cubic, leaveTangent=0.627, time=0.3676, value=0.55}, <at 0.632 0.55>, <at 0.8525 0.3>]}
       GreenCurve {keys [<at 0.2333 0.3>, {arriveTangent=0.627, interpMode=RCIM_Cubic, leaveTangent=0.627, time=0.3676, value=0.55}, <at 0.632 0.55>, <at 0.8525 0.3>]}
       RedCurve {keys [<at 0.2333 0.3>, {arriveTangent=0.627, interpMode=RCIM_Cubic, leaveTangent=0.627, time=0.3676, value=0.55}, <at 0.632 0.55>, <at 0.8525 0.3>]}
      }}
     CurveIndex=$Particles.NormalizedRibbonPos
    )
    Scale Color=&VectorFromFloat(Value=&Lerp_Float(A=1, B=2, Alpha=$System.SmoothedForwardDot))
   )
   /SetVariables(Particles.BackSpiralColor:LinearColor=&MultiplyLinearColor(
     Linear Color=$Particles.Color
     Scale Factor=&MakeVector4_FromLinearColor(LinearColor=<rgba 0.5677 0.4789 0.3902 1>)
    )) {isSetParameters=true}
   /SetVariables(Particles.SpawnAlpha:NiagaraFloat=&FloatFromCurve(FloatCurve=DataInterface() {propertyValues {Curve {keys [<at 0 0>,<at 1 1>]}}}, CurveIndex=$Particles.Age)) {isSetParameters=true, module=SetVariables_#13}
   /SetVariables(Particles.GoalPos:NiagaraPosition=&AddVectorToPosition(
     Position=$Particles.SkeletalMeshLocation.SampledPosition
     Vector=&Lerp_Vector(
      A=<0 0 0>
      B=&Lerp_Vector(
       A=<0 0 0>
       B=&Lerp_Vector(A=<0 175 0>, B=<0 215 0>, Alpha=$System.SmoothedForwardDot)
       Alpha=$Particles.SpawnAlpha
      )
      Alpha=$Particles.NormalizedRibbonPos
     )
    )) {isSetParameters=true, module=SetVariables_#14}
   /SetVariables(Particles.SpiralStrength:NiagaraFloat=&Multiply_Float(
     A=&Multiply_Float(
      A=1.5
      B=&FloatFromCurve(
       FloatCurve=DataInterface() {propertyValues {Curve {keys [
           {arriveTangent=0.5274, interpMode=RCIM_Cubic, leaveTangent=0.5274, tangentMode=RCTM_User, time=0.2, value=-0.025}
           {arriveTangent=1.8506, interpMode=RCIM_Cubic, leaveTangent=1.8506, tangentMode=RCTM_User, time=0.4, value=1}
           {arriveTangent=-2.6462, interpMode=RCIM_Cubic, leaveTangent=-2.6462, tangentMode=RCTM_User, time=0.7, value=0.5}
           {arriveTangent=-0.4763, interpMode=RCIM_Cubic, leaveTangent=-0.4763, tangentMode=RCTM_User, time=0.9, value=0.3}
          ]}}}
       CurveIndex=$Particles.NormalizedRibbonPos
      )
     )
     B=$Particles.SpawnAlpha
    )) {isSetParameters=true}
   /SetVariables(Particles.SpiralTime:NiagaraFloat=&Add_Float(
     A=&Add_Float(A=&Multiply_Float(A=$Emitter.Age, B=-1.2), B=$Particles.NormalizedRibbonPos)
     B=&Multiply_FloatByInt(
      Float=&Divide_Float(A=$Particles.SpiralFreq, B=&MakeFloatFromInt(Int=$Emitter.RibbonNumber))
      Integer=$Particles.PartitionParticles.Partition
     )
    )) {isSetParameters=true}
   /SetVariables(Particles.SpiralOffset:Vector3f=&MakeVector(
     X=&Sine(Normalized Angle=$Particles.SpiralTime, Period=$Particles.SpiralFreq, Scale=$Particles.SpiralStrength, Bias=0)
     Y=0
     Z=&Cosine(Normalized Angle=$Particles.SpiralTime, Period=$Particles.SpiralFreq, Scale=$Particles.SpiralStrength, Bias=0)
    )) {isSetParameters=true}
   &InheritSourceMovement(
    Applied Position Scale=&VectorFromFloat(Value=&Multiply_Float(
      A=-0.5
      B=&FloatFromCurve(
       FloatCurve=DataInterface() {propertyValues {Curve {keys [
           {arriveTangent=-0.0194, interpMode=RCIM_Cubic, leaveTangent=-0.0194, tangentMode=RCTM_User, time=0.2, value=0}
           {arriveTangent=1.6438, interpMode=RCIM_Cubic, leaveTangent=1.6438, tangentMode=RCTM_User, time=1, value=1}
          ]}}}
       CurveIndex=$Particles.NormalizedRibbonPos
      )
     ))
    ApplyForce=false
    ApplyPosition=true
    ApplyPositionFalloff=false
    Source Velocity:Vector3f=&TransformVector(Vector=$Engine.Owner.Velocity, Source Space=::World, Destination Space=::Local)
   )
   &SolveForcesAndVelocity(Clamp Velocity=false, Limit Acceleration=false)
   /SetVariables(Particles.Position=&AddVectorToPosition(
     Position=&Lerp_Vector(
      A=Linked() {linkedVariable {name=Particles.Position, type=NiagaraPosition}}
      B=Linked() {linkedVariable {name=Particles.GoalPos, type=NiagaraPosition}}
      Alpha=&Multiply_Float(A=$Engine.DeltaTime, B=10)
     )
     Vector=$Particles.SpiralOffset
    )) {isSetParameters=true, module=SetVariables_#15}
   /SetVariables(Particles.RibbonFacing=&NormalizeVector(
     Vector To Normalize=&TransformVector(Vector=$Particles.SpiralOffset, Destination Space=::Simulation)
     Safe Normalize=true
     Safe Fallback Threshold=0
     Safe Fallback Vector=<0 0 0.001>
    )) {isSetParameters=true}
   /SetVariables(Particles.BackRibbonFacing:Vector3f=&NormalizeVector(
     Vector To Normalize=&TransformVector(Vector=$Particles.SpiralOffset, Destination Space=::Simulation)
     Safe Normalize=true
     Safe Fallback Threshold=0
     Safe Fallback Vector=<0 0 0.001>
    )) {isSetParameters=true}
   &ScaleRibbonWidth(Ribbon Width Scale=&FloatFromCurve(
     FloatCurve=DataInterface() {propertyValues {Curve {keys [
         {arriveTangent=7.9812, interpMode=RCIM_Cubic, leaveTangent=7.9812, tangentMode=RCTM_User, time=0.25, value=0.45}
         {interpMode=RCIM_Cubic, tangentMode=RCTM_User, time=0.4, value=1.4}
         {arriveTangent=-1.6291, interpMode=RCIM_Cubic, leaveTangent=-1.6291, tangentMode=RCTM_User, time=1, value=0.5}
        ]}}}
     CurveIndex=$Particles.NormalizedRibbonPos
     Scale Curve=$Particles.SpawnAlpha
    ))
   &DynamicMaterialParameters(
    Index 0 Param 1=$Particles.UVOffset
    Index 0 Param 2=$System.SmoothedForwardDot
    Module.Index0Param0Write=true
    Module.Index0Param1Write=true
    Module.Index0Param2Write=true
    Module.Index0Param3Write=true
   )
  ]
  rendererClasses [&NiagaraRibbonRendererProperties]
  renderers [
   {rendererClass=&NiagaraRibbonRendererProperties}
   {rendererClass=&NiagaraRibbonRendererProperties, rendererIndex=1}
  ]
 }
 NE_StraightRibbon {
  emitterUpdateScript [
   &EmitterState(
    Inactive Response=::NewEnumerator1
    Life Cycle Mode=::NewEnumerator1
    Loop Behavior=::NewEnumerator1
    Loop Duration Mode=::NewEnumerator1
    UseLoopDelay=false
   )
   &SpawnBurst_Instantaneous(Spawn Count=20, Spawn Time=1, Use Spawn Probability=false, UseLoopCountLimit=false)
  ]
  particleSpawnScript [
   &InitializeParticle(Lifetime=5, Ribbon Width=10, Ribbon Width Mode=::NewEnumerator1, UsePositionOffset=false, Write Material Random=true)
   /SetVariables(Particles.NormalizedRibbonPos:NiagaraFloat=&OneMinusFloat(Float=&ReturnNormalizedExecIndex())) {isSetParameters=true, module=SetVariables_#10}
   /SetVariables(Particles.RibbonLinkOrder:NiagaraFloat=$Particles.NormalizedRibbonPos) {isSetParameters=true}
   /SetVariables(Particles.GlowColor=&ColorFromCurve(
     ColorCurve=DataInterface() {propertyValues {
       AlphaCurve {keys [<at 0 1 RCIM_Cubic>,<at 0.6603 1>,<at 1 0>]}
       BlueCurve {keys [<at 0 1 RCIM_Cubic>,<at 0.3054 0.575>]}
       GreenCurve {keys [<at 0 1 RCIM_Cubic>,<at 0.3054 0.575>]}
       RedCurve {keys [<at 0 1 RCIM_Cubic>,<at 0.3054 0.575>]}
      }}
     CurveIndex=$Particles.NormalizedRibbonPos
    )) {isSetParameters=true}
   /SetVariables() {isSetParameters=true, module=SetVariables_#11}
  ]
  particleUpdateScript [
   &ParticleState(Kill Particles When Lifetime Has Elapsed=false)
   &SkeletalMeshLocation(Bone Sampling Mode=::NewEnumerator4, MaskSpawnGroup=false, Position Sampling=::NewEnumerator2, Skeletal Mesh=$System.Skeletal Mesh)
   /SetVariables(Particles.UVOffset:NiagaraFloat=&Add_Float(
     A=$Particles.UVOffset
     B=&Multiply_Float(A=$Engine.DeltaTime, B=&Lerp_Float(A=-0.1, B=-1.5, Alpha=$System.SmoothedForwardDot))
    )) {isSetParameters=true, module=SetVariables_#12}
   &Color(
    Color=&ColorFromCurve(
     ColorCurve=DataInterface() {propertyValues {
       AlphaCurve {keys [
         <at 0.1909 0>
         {arriveTangent=2.4905, interpMode=RCIM_Cubic, leaveTangent=2.4905, time=0.4114, value=1}
         <at 0.5924 1>
         <at 0.8624 0 RCIM_Cubic>
        ]}
       BlueCurve {keys [<at 0.2333 0.3>, {arriveTangent=0.627, interpMode=RCIM_Cubic, leaveTangent=0.627, time=0.3676, value=0.55}, <at 0.632 0.55>, <at 0.8525 0.3>]}
       GreenCurve {keys [<at 0.2333 0.3>, {arriveTangent=0.627, interpMode=RCIM_Cubic, leaveTangent=0.627, time=0.3676, value=0.55}, <at 0.632 0.55>, <at 0.8525 0.3>]}
       RedCurve {keys [<at 0.2333 0.3>, {arriveTangent=0.627, interpMode=RCIM_Cubic, leaveTangent=0.627, time=0.3676, value=0.55}, <at 0.632 0.55>, <at 0.8525 0.3>]}
      }}
     CurveIndex=$Particles.NormalizedRibbonPos
    )
    Scale Color=&VectorFromFloat(Value=&Lerp_Float(A=1, B=2, Alpha=$System.SmoothedForwardDot))
   )
   /SetVariables(Particles.SpawnAlpha:NiagaraFloat=&FloatFromCurve(FloatCurve=DataInterface() {propertyValues {Curve {keys [<at 0 0>,<at 1 1>]}}}, CurveIndex=$Particles.Age)) {isSetParameters=true, module=SetVariables_#13}
   /SetVariables(Particles.GoalPos:NiagaraPosition=&AddVectorToPosition(
     Position=$Particles.SkeletalMeshLocation.SampledPosition
     Vector=&Lerp_Vector(
      A=<0 0 0>
      B=&Lerp_Vector(
       A=<0 0 0>
       B=&Lerp_Vector(
        A=<0 175 0>
        B=<0 215 0>
        Alpha=&FN_FrontEndFloatModifier(Gameplay=$System.SmoothedForwardDot, Front End=2)
       )
       Alpha=$Particles.SpawnAlpha
      )
      Alpha=$Particles.NormalizedRibbonPos
     )
    )) {isSetParameters=true, module=SetVariables_#14}
   &InheritSourceMovement(
    Applied Position Scale=&VectorFromFloat(Value=&Multiply_Float(
      A=-0.5
      B=&FloatFromCurve(
       FloatCurve=DataInterface() {propertyValues {Curve {keys [
           {arriveTangent=-0.0194, interpMode=RCIM_Cubic, leaveTangent=-0.0194, tangentMode=RCTM_User, time=0.2, value=0}
           {arriveTangent=1.6438, interpMode=RCIM_Cubic, leaveTangent=1.6438, tangentMode=RCTM_User, time=1, value=1}
          ]}}}
       CurveIndex=$Particles.NormalizedRibbonPos
      )
     ))
    ApplyForce=false
    ApplyPosition=true
    ApplyPositionFalloff=false
    Source Velocity:Vector3f=&TransformVector(Vector=$Engine.Owner.Velocity, Source Space=::World, Destination Space=::Local)
   )
   &SolveForcesAndVelocity(Clamp Velocity=false, Limit Acceleration=false)
   /SetVariables(Particles.Position=&Lerp_Vector(
     A=Linked() {linkedVariable {name=Particles.Position, type=NiagaraPosition}}
     B=Linked() {linkedVariable {name=Particles.GoalPos, type=NiagaraPosition}}
     Alpha=&Multiply_Float(A=$Engine.DeltaTime, B=10)
    )) {isSetParameters=true, module=SetVariables_#15}
   &ScaleRibbonWidth(Ribbon Width Scale=&FloatFromCurve(
     FloatCurve=DataInterface() {propertyValues {Curve {keys [
         {arriveTangent=7.9812, interpMode=RCIM_Cubic, leaveTangent=7.9812, tangentMode=RCTM_User, time=0.25, value=0.45}
         {interpMode=RCIM_Cubic, tangentMode=RCTM_User, time=0.4, value=1.4}
         {arriveTangent=-1.6291, interpMode=RCIM_Cubic, leaveTangent=-1.6291, tangentMode=RCTM_User, time=1, value=0.5}
        ]}}}
     CurveIndex=$Particles.NormalizedRibbonPos
     Scale Curve=$Particles.SpawnAlpha
    ))
   /SetVariables(Particles.GlowWidth:NiagaraFloat=&Multiply_Float(A=$Particles.RibbonWidth, B=2.2)) {isSetParameters=true}
   &DynamicMaterialParameters(
    Index 0 Param 1=$Particles.UVOffset
    Index 0 Param 2=&FN_FrontEndFloatModifier(Gameplay=$System.SmoothedForwardDot, Front End=2)
    Module.Index0Param0Write=true
    Module.Index0Param1Write=true
    Module.Index0Param2Write=true
    Module.Index0Param3Write=true
   )
   /SetVariables(Particles.GlowColor=&MultiplyLinearColor(
     Linear Color=$Particles.Initial.GlowColor
     Scale Factor=&MakeVector4_FromLinearColor(LinearColor=$User.StraightColorInner)
    )) {isSetParameters=true}
  ]
  rendererClasses [&NiagaraRibbonRendererProperties]
  renderers [
   {rendererClass=&NiagaraRibbonRendererProperties}
   {rendererClass=&NiagaraRibbonRendererProperties, rendererIndex=1}
  ]
 }
 Speed_Lines {
  emitterUpdateScript [
   &EmitterState()
   &SpawnRate(SpawnRate=&FN_FrontEndFloatModifier(Gameplay=100, Front End=50), Use Spawn Probability=false) {module=SpawnRate001}
  ]
  particleSpawnScript [
   &InitializeParticle(
    Color=&MakeLinearColorFromVectorAndFloat("Vector (RGB)"=<1 1 1>, "Float (Alpha)"=&FN_FrontEndFloatModifier(Gameplay=0.25, Front End=0.1))
    Lifetime Max=0.16
    Lifetime Min=0.1
    Lifetime Mode=::NewEnumerator1
    Sprite Size Max=<4 10>
    Sprite Size Min=<3 5>
    Sprite Size Mode=::NewEnumerator2
    UsePositionOffset=true
    Write Material Random=true
   )
   &ShapeLocation(
    Handle Radius=5
    Large Radius=&FN_FrontEndFloatModifier(Gameplay=250, Front End=200)
    MaskSpawnGroup=false
    Offset=&FN_FrontEndVector3Modifier(Gameplay=<0 -200 15>, Front End=<0 -200 15>)
    Offset Mode=::NewEnumerator0
    Override Seed=false
    Rotation Axis=<0 1 0>
    Rotation Mode=::NewEnumerator1
    Shape Primitive=::NewEnumerator3
    UseOwnerScale=false
   )
   /SetVariables(Particles.Position=&Multiply_Vector(A=Linked() {linkedVariable {name=Particles.Position, type=NiagaraPosition}}, B=<1.2 1 0.8>)) {isSetParameters=true}
   &AddVelocityFromPoint(
    Module.Use Falloff Distance=false
    Origin Offset=<0 -1000 -250>
    Velocity Strength=&FN_FrontEndFloatModifier(Gameplay=1000, Front End=0)
   )
   &AddVelocity(
    Velocity=&FN_FrontEndVector3Modifier(
     Gameplay=&Add_Vector(A=<0 300 0>, B=&Multiply_VectorByFloat(Vector=<0 0 222>, Float=$System.SmoothedForwardDot))
     Front End=<0 100 8>
    )
    Velocity Speed Scale=&FN_FrontEndFloatModifier(Gameplay=-1, Front End=1)
   )
  ]
  particleUpdateScript [
   &ParticleState()
   &ScaleSpriteSize(Scale Factor=&Vector2DFromCurve(
     Vector2Curve=DataInterface() {propertyValues {XCurve {keys [<at 0 0.5 RCIM_Cubic>,<at 0.5 1 RCIM_Cubic>]}, YCurve {keys [<at 0 0 RCIM_Cubic>,<at 0.5 1 RCIM_Cubic>]}}}
     ScaleCurve=&Vector2DFromFloat(Value=1)
    ))
   &SolveForcesAndVelocity(Clamp Velocity=false, Limit Acceleration=false)
   &ScaleSpriteSizeBySpeed(Max Scale Factor=<1 50>, Min Scale Factor=<0.01 0.05>, Velocity Threshold=1)
   &ScaleColor(
    Scale Alpha=&FloatFromCurve(
     FloatCurve=DataInterface() {propertyValues {Curve {keys [
         {arriveTangent=6.8397, interpMode=RCIM_Cubic, leaveTangent=6.8397, tangentMode=RCTM_User, time=0, value=0}
         <at 0.3 1 RCIM_Cubic>
         {arriveTangent=-2.221, interpMode=RCIM_Cubic, leaveTangent=-2.221, tangentMode=RCTM_User, time=1, value=0}
        ]}}}
     Scale Curve=&FN_FrontEndFloatModifier(
      Gameplay=&Lerp_Float(
       A=0
       B=&FNTimeOfDayScalar(Night Scale=0.2, Morning Scale=0.3, Day Scale=0.4, Evening Scale=0.3)
       Alpha=&Multiply_Float(A=$System.SystemFadeIn, B=$System.Timeline.Time)
      )
      Front End=&Lerp_Float(A=0, B=0.4, Alpha=$System.SystemFadeIn)
     )
    )
    Scale RGB=&VectorFromFloat(Value=&FN_FrontEndFloatModifier(Gameplay=1, Front End=1))
    ScaleA=true
    ScaleRGB=true
   )
  ]
  rendererClasses [&NiagaraSpriteRendererProperties]
  renderers [{rendererClass=&NiagaraSpriteRendererProperties}]
 }
}
summary {
 systemName=NS_Glider_LiftingAura_Trail
 emitters [
  {emitterName=NE_StraightRibbon, rendererClasses [&NiagaraRibbonRendererProperties]}
  {emitterName=NE_HeartGlow, rendererClasses [&NiagaraSpriteRendererProperties]}
  {emitterName=NE_SpiralRibbon, rendererClasses [&NiagaraRibbonRendererProperties]}
  {emitterName=NE_Light, rendererClasses [&NiagaraLightRendererProperties]}
  {emitterName=Speed_Lines, rendererClasses [&NiagaraSpriteRendererProperties]}
 ]
 userVariables [
  User.bIsFrontEnd {type=NiagaraBool}
  User.bIsFrontEndPreview {type=NiagaraBool}
  User.bIsFullyDeployed {type=NiagaraBool}
  User.Debug {type=NiagaraBool}
  User.ForwardDot {type=NiagaraFloat, defaultValue=0, description=""}
  User.RightDot {type=NiagaraFloat, defaultValue=0, description=""}
  User.SpiralColorInner {type=LinearColor, description="", defaultValue {struct=LinearColor, value=<rgba 0.9473 0.1258 0 1>}}
  User.SpiralColorMid {type=LinearColor, description="", defaultValue {struct=LinearColor, value=<rgba 1 0.2811 0.0208 1>}}
  User.SpiralMidColor {type=LinearColor, description="", defaultValue {struct=LinearColor, value=<rgba 1 1 1 1>}}
  User.StraightColorInner {type=LinearColor, description="", defaultValue {struct=LinearColor, value=<rgba 1 0.6654 0.1301 1>}}
  User.StraightColorMid {type=LinearColor, description="", defaultValue {struct=LinearColor, value=<rgba 0.8594 0.2939 0.0582 1>}}
 ]
}
